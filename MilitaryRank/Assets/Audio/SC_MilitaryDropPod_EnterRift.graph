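# Sound cue for the military drop pod entering a rift, mixing shared parent drop pod sounds with eagle whistles.
# A mixer layers a pitch-modulated random EnterRiftC variant, a random EnterRiftB variant and a much quieter random whistle.
# Two empty slots in the whistle randomizer leave that layer silent some of the time.
# It borrows the parent pod's high-frequency alert attenuation, overrides concurrency and disables virtualization.
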SoundCue /MilitaryRank/Assets/Audio/SC_MilitaryDropPod_EnterRift
root=Mixer_0
nodes [
 {id=Mixer_0, children [Modulator_0,Random_1,Random_3], properties {inputVolume [1,1,0.2]}}
 {id=Modulator_0, children [Random_0], properties {pitchMax=1.3, pitchMin=0.9, volumeMax=1, volumeMin=1}}
 {id=Random_0, children [WavePlayer_5,WavePlayer_6], properties {bShouldExcludeFromBranchCulling=true, weights [1,1]}}
 {id=Random_1, children [WavePlayer_2,WavePlayer_4], properties {weights [1,1]}}
 {id=Random_3, children [WavePlayer_8, WavePlayer_9, WavePlayer_10, WavePlayer_11, null, null], properties {weights [1,1,1,1,1,1]}}
 {id=WavePlayer_10, properties {soundWaveAssetPtr=/MilitaryRank/Assets/Audio/Wavs/TidalCraneDropPod_EagleWhistle_06}}
 {id=WavePlayer_11, properties {soundWaveAssetPtr=/MilitaryRank/Assets/Audio/Wavs/TidalCraneDropPod_EagleWhistle_07}}
 {id=WavePlayer_2, properties {soundWaveAssetPtr=/DropPodParentGameplay/Assets/Audio/wav/TidalCraneDropPod_EnterRiftB_01}}
 {id=WavePlayer_4, properties {soundWaveAssetPtr=/DropPodParentGameplay/Assets/Audio/wav/TidalCraneDropPod_EnterRiftB_02}}
 {id=WavePlayer_5, properties {soundWaveAssetPtr=/DropPodParentGameplay/Assets/Audio/wav/TidalCraneDropPod_EnterRiftC_01}}
 {id=WavePlayer_6, properties {soundWaveAssetPtr=/DropPodParentGameplay/Assets/Audio/wav/TidalCraneDropPod_EnterRiftC_02}}
 {id=WavePlayer_8, properties {soundWaveAssetPtr=/MilitaryRank/Assets/Audio/Wavs/TidalCraneDropPod_EagleWhistle_04}}
 {id=WavePlayer_9, properties {soundWaveAssetPtr=/MilitaryRank/Assets/Audio/Wavs/TidalCraneDropPod_EagleWhistle_05}}
]
properties {
 attenuationSettings=/DropPodParentGameplay/Assets/Audio/ATT_TidalCraneDropPod_Alert_HighFreq
 bOverrideConcurrency=true
 maxDistance=230000
 soundClassObject=&Fort_Ambient_Athena
 virtualizationMode=Disabled
 volumeMultiplier=0.9
 concurrencyOverrides {maxCount=4}
}
