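# Sound cue for the military drop pod flare activation, layering landing, activation and whistle sounds.
# A Mixer combines a random pick of three landing variants, a fixed activation wave at reduced volume, and a random pick of four whistle variants.
# Both Random nodes weight their variants equally.
# The cue borrows flare gun alert attenuation and a third-person ranged-weapon fire sound class, with an override capping concurrent instances.
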SoundCue /MilitaryRank/Assets/Audio/SC_MilitaryDropPod_FlareActivate
root=Mixer_0
nodes [
 {id=Mixer_0, children [Random_1,WavePlayer_1,Random_0], properties {inputVolume [1,0.5,1]}}
 {id=Random_0, children [WavePlayer_2,WavePlayer_6,WavePlayer_7,WavePlayer_8], properties {weights [1,1,1,1]}}
 {id=Random_1, children [WavePlayer_3,WavePlayer_4,WavePlayer_5], properties {weights [1,1,1]}}
 {id=WavePlayer_1, properties {soundWaveAssetPtr=/MilitaryRank/Assets/Audio/Wavs/TidalCraneDropPod_FlareActivate}}
 {id=WavePlayer_2, properties {soundWaveAssetPtr=/MilitaryRank/Assets/Audio/Wavs/TidalCraneDropPod_FlareWhistle_01}}
 {id=WavePlayer_3, properties {soundWaveAssetPtr=/MilitaryRank/Assets/Audio/Wavs/TidalCraneDropPod_FlareActivateLand_03}}
 {id=WavePlayer_4, properties {soundWaveAssetPtr=/MilitaryRank/Assets/Audio/Wavs/TidalCraneDropPod_FlareActivateLand_02}}
 {id=WavePlayer_5, properties {soundWaveAssetPtr=/MilitaryRank/Assets/Audio/Wavs/TidalCraneDropPod_FlareActivateLand_01}}
 {id=WavePlayer_6, properties {soundWaveAssetPtr=/MilitaryRank/Assets/Audio/Wavs/TidalCraneDropPod_FlareWhistle_02}}
 {id=WavePlayer_7, properties {soundWaveAssetPtr=/MilitaryRank/Assets/Audio/Wavs/TidalCraneDropPod_FlareWhistle_03}}
 {id=WavePlayer_8, properties {soundWaveAssetPtr=/MilitaryRank/Assets/Audio/Wavs/TidalCraneDropPod_FlareWhistle_04}}
]
properties {
 attenuationSettings=/Game/Sounds/Attenuation/Fort_Weapon_Projectile_FlareGun_Alert_Att
 maxDistance=10000
 soundClassObject=&Fort_WeaponRanged_Fire_3P_PSM
 volumeMultiplier=0.35
 concurrencyOverrides {maxCount=4}
}
