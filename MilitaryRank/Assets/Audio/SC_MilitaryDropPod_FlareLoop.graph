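# Sound cue for the looping flare sound on the military drop pod, likely heard while the pod's flare burns.
# A single wave player outputs the TidalCraneDropPod flare loop wave directly, with no mixing or randomisation.
# It uses the close-range premium supply drop flare attenuation, plays at reduced volume, and is routed through the Athena ambient sound class.
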SoundCue /MilitaryRank/Assets/Audio/SC_MilitaryDropPod_FlareLoop
root=WavePlayer_1
nodes [{id=WavePlayer_1, properties {soundWaveAssetPtr=/MilitaryRank/Assets/Audio/Wavs/TidalCraneDropPod_Flare_Loop}}]
properties {
 attenuationSettings=&/ATT_SC_PremiumSupplyDrop_FlareClose
 maxDistance=1500
 soundClassObject=&Fort_Ambient_Athena
 volumeMultiplier=0.6
}
