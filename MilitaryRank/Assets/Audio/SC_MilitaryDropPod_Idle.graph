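# Sound cue for the idle ambience of a military drop pod, reusing premium supply drop audio.
# A looping idle bed, a looped random pick from five air-release hisses, and a half-volume servo motor loop are mixed together.
# An enveloper on the mix fades the volume in from silence.
# It uses chest attenuation and the gadgets sound class, with concurrency set to prevent new instances at the limit.
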SoundCue /MilitaryRank/Assets/Audio/SC_MilitaryDropPod_Idle
root=Enveloper_3
nodes [
 {
  id=Enveloper_3
  children [Mixer_2]
  properties {pitchCurve {editorCurveData {keys [{time=0}]}}, volumeCurve {editorCurveData {keys [<at 0 0 RCIM_Cubic>,<at 1 1>]}}}
 }
 {id=Looping_0, children [Random_1]}
 {id=Mixer_2, children [WavePlayer_0, Looping_0, WavePlayer_18], properties {inputVolume [1,1,0.5]}}
 {id=Random_1, children [WavePlayer_13, WavePlayer_14, WavePlayer_15, WavePlayer_16, WavePlayer_17], properties {weights [1,1,1,1,1]}}
 {id=WavePlayer_0, properties {bLooping=true, soundWaveAssetPtr=/MilitaryRank/Assets/Audio/Wavs/PremiumSupplyDrop_IdleLoop}}
 {id=WavePlayer_13, properties {soundWaveAssetPtr=/MilitaryRank/Assets/Audio/Wavs/PremiumSupplyDrop_AirRelease_01}}
 {id=WavePlayer_14, properties {soundWaveAssetPtr=/MilitaryRank/Assets/Audio/Wavs/PremiumSupplyDrop_AirRelease_02}}
 {id=WavePlayer_15, properties {soundWaveAssetPtr=/MilitaryRank/Assets/Audio/Wavs/PremiumSupplyDrop_AirRelease_03}}
 {id=WavePlayer_16, properties {soundWaveAssetPtr=/MilitaryRank/Assets/Audio/Wavs/PremiumSupplyDrop_AirRelease_04}}
 {id=WavePlayer_17, properties {soundWaveAssetPtr=/MilitaryRank/Assets/Audio/Wavs/PremiumSupplyDrop_AirRelease_05}}
 {id=WavePlayer_18, properties {bLooping=true, soundWaveAssetPtr=/MilitaryRank/Assets/Audio/Wavs/Spot_Light_Servo_Motor_Mechanisim_Loop}}
]
properties {
 attenuationSettings=/Game/Sounds/Attenuation/Fort_Chests_Att
 maxDistance=1300
 soundClassObject=&Fort_Gadgets
 volumeMultiplier=0.85
 concurrencyOverrides {resolutionRule=PreventNew}
}
