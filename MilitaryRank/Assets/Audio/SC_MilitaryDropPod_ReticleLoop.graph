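# Sound cue for the looping target reticle of the military drop pod, spatialized with supply drop marker attenuation in the gadgets sound class.
# The first two quality levels mix the target loop with two quieter looping layers of random Bling one-shots, each varied in pitch and volume.
# One layer inserts a random delay of a few seconds between shots.
# The other two quality levels play only an optimized target loop.
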SoundCue /MilitaryRank/Assets/Audio/SC_MilitaryDropPod_ReticleLoop
root=QualityLevel_0
nodes [
 {id=Delay_0, children [Random_3], properties {delayMax=3, delayMin=2}}
 {id=Looping_0, children [Random_0]}
 {id=Looping_1, children [Delay_0]}
 {id=Mixer_0, children [WavePlayer_8,Modulator_1,Modulator_0], properties {inputVolume [1,0.35,0.35]}}
 {id=Modulator_0, children [Looping_1], properties {pitchMax=1.2, pitchMin=0.8, volumeMin=0.7}}
 {id=Modulator_1, children [Looping_0], properties {pitchMax=1.2, pitchMin=0.8, volumeMin=0.7}}
 {id=QualityLevel_0, children [Mixer_0,Mixer_0,WavePlayer_9,WavePlayer_9]}
 {id=Random_0, children [WavePlayer_0,WavePlayer_1,WavePlayer_2,WavePlayer_3], properties {weights [1,1,1,1]}}
 {id=Random_3, children [WavePlayer_4,WavePlayer_5,WavePlayer_6,WavePlayer_7], properties {weights [1,1,1,1]}}
 {id=WavePlayer_0, properties {soundWaveAssetPtr=/MilitaryRank/Assets/Audio/Wavs/Bling_Bus_OS_01}}
 {id=WavePlayer_1, properties {soundWaveAssetPtr=/MilitaryRank/Assets/Audio/Wavs/Bling_Bus_OS_02}}
 {id=WavePlayer_2, properties {soundWaveAssetPtr=/MilitaryRank/Assets/Audio/Wavs/Bling_Bus_OS_03}}
 {id=WavePlayer_3, properties {soundWaveAssetPtr=/MilitaryRank/Assets/Audio/Wavs/Bling_Bus_OS_04}}
 {id=WavePlayer_4, properties {soundWaveAssetPtr=/MilitaryRank/Assets/Audio/Wavs/Bling_Bus_OS_05}}
 {id=WavePlayer_5, properties {soundWaveAssetPtr=/MilitaryRank/Assets/Audio/Wavs/Bling_Bus_OS_06}}
 {id=WavePlayer_6, properties {soundWaveAssetPtr=/MilitaryRank/Assets/Audio/Wavs/Bling_Bus_OS_07}}
 {id=WavePlayer_7, properties {soundWaveAssetPtr=/MilitaryRank/Assets/Audio/Wavs/Bling_Bus_OS_08}}
 {id=WavePlayer_8, properties {soundWaveAssetPtr=/MilitaryRank/Assets/Audio/Wavs/sfx_supplydrop_target_loop}}
 {id=WavePlayer_9, properties {soundWaveAssetPtr=/MilitaryRank/Assets/Audio/Wavs/sfx_supplydrop_target_loop_opt}}
]
properties {
 attenuationSettings=&/Fort_SupplyDrop_Marker_Close_Athena_Att
 maxDistance=1500
 soundClassObject=&Fort_Gadgets
 volumeMultiplier=1
}
