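# Sound cue for the military drop pod's searching sound, built from reused PremiumSupplyDrop waves.
# A mixer layers a random opening click, a random servo, a pitch-modulated scan-success tone and a landing-mech wave, each at its own volume.
# No loop node or looping flag appears in the cue, despite its SearchingLoop name.
# It uses mono item-interact attenuation, the loot-collect sound class and a concurrency override that prevents new instances at the limit.
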SoundCue /MilitaryRank/Assets/Audio/SC_MilitaryDropPod_SearchingLoop
root=Mixer_0
nodes [
 {id=Mixer_0, children [Random_1,Random_4,Modulator_0,WavePlayer_0], properties {inputVolume [0.3,0.5,0.2,0.05]}}
 {id=Modulator_0, children [WavePlayer_8], properties {pitchMin=1}}
 {id=Random_1, children [WavePlayer_4,WavePlayer_5,WavePlayer_6,WavePlayer_7], properties {weights [1,1,1,1]}}
 {id=Random_4, children [WavePlayer_9, WavePlayer_10, WavePlayer_11], properties {weights [1,1,1]}}
 {id=WavePlayer_0, properties {soundWaveAssetPtr=/MilitaryRank/Assets/Audio/Wavs/PremiumSupplyDrop_LandingMech}}
 {id=WavePlayer_10, properties {soundWaveAssetPtr=/MilitaryRank/Assets/Audio/Wavs/PremiumSupplyDrop_Servo_02}}
 {id=WavePlayer_11, properties {soundWaveAssetPtr=/MilitaryRank/Assets/Audio/Wavs/PremiumSupplyDrop_Servo_03}}
 {id=WavePlayer_4, properties {soundWaveAssetPtr=/MilitaryRank/Assets/Audio/Wavs/PremiumSupplyDrop_OpeningClick_01}}
 {id=WavePlayer_5, properties {soundWaveAssetPtr=/MilitaryRank/Assets/Audio/Wavs/PremiumSupplyDrop_OpeningClick_02}}
 {id=WavePlayer_6, properties {soundWaveAssetPtr=/MilitaryRank/Assets/Audio/Wavs/PremiumSupplyDrop_OpeningClick_03}}
 {id=WavePlayer_7, properties {soundWaveAssetPtr=/MilitaryRank/Assets/Audio/Wavs/PremiumSupplyDrop_OpeningClick_04}}
 {id=WavePlayer_8, properties {soundWaveAssetPtr=/MilitaryRank/Assets/Audio/Wavs/PremiumSupplyDrop_Scan_Success}}
 {id=WavePlayer_9, properties {soundWaveAssetPtr=/MilitaryRank/Assets/Audio/Wavs/PremiumSupplyDrop_Servo_01}}
]
properties {
 attenuationSettings=/Game/Sounds/Attenuation/Fort_Item_Interact_Mono_Att
 maxDistance=1500
 soundClassObject=&Fort_Loot_Collect
 volumeMultiplier=0.35
 concurrencyOverrides {resolutionRule=PreventNew}
}
